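# Translucent UI material for the Demon Mask Void reticle, drawing a spiral, diamond and drop lines.
# Channel-packed distance-field textures supply fills, outlines and soft shadow glows, layered through named reroutes into color and opacity with per-part and global opacity controls.
# The spiral's rotation blends a set angle with time-driven spin via SpeedRotation.
# IsPortalActive blends between two color sets and changes the spiral's scale.
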Material /DemonMaskVoid/UI/Materials/M_UI_VoidMaskReticle
graph {
 properties {bAutomaticallySetUsageInEditor=false, blendMode=BLEND_Translucent, materialDomain=MD_UI, usedShadingModels=Default Lit}
 nodes [
  n1 = NamedRerouteUsage()
  n2 = NamedRerouteUsage()
  n3 = NamedRerouteUsage()
  n4 = NamedRerouteUsage()
  n5 = LinearInterpolate(A=NamedRerouteUsage(), B=NamedRerouteUsage(), Alpha=NamedRerouteUsage())
  n6 = NamedRerouteUsage()
  n7 = NamedRerouteUsage()
  n8 = NamedRerouteUsage()
  n9 = LinearInterpolate(A=NamedRerouteUsage(), B=@n8, Alpha=@n8)
  n10 = NamedRerouteUsage()
  LinearInterpolate(
   A=LinearInterpolate(B=NamedRerouteUsage(), Alpha=Add(A=@n1, B=@n2))
   B=LinearInterpolate(A=Multiply(A=@n3, B=@n4), B=LinearInterpolate(A=@n5, B=@n6, Alpha=@n7), Alpha=Add(A=@n9, B=@n10)) {constAlpha=0}
   Alpha=ScalarParameter() {parameterName=IsPortalActive}
  )
  n11 = TextureCoordinate()
  n12 = ComponentMask(None=@n11) {g=true}
  n13 = ScalarParameter() {defaultValue=0.196, parameterName=OuterShapePosition}
  n14 = MaterialFunctionCall(
   UVs=AppendVector(A=ComponentMask(None=@n11) {r=true}, B=Add(A=@n12, B=@n13) {constB=0})
   Texture Scale=AppendVector(A=ScalarParameter() {parameterName=ScaleXDiamond}, B=ScalarParameter() {parameterName=ScaleYDiamond})
  ) {materialFunction=&ScaleUVsByCenter}
  n15 = TextureSample(UVs=@n14.UVs) {samplerSource=SSM_Clamp_WorldGroupSettings, texture=/DemonMaskVoid/UI/Textures/T_UI_Reticle_DemonVoid_DiamondStroke}
  MaterialFunctionCall!(RGBA Channels=@n15.R, Glow Min=Constant() {r=0.3}, Glow Max=Constant() {r=1}, Outline Thickness) {materialFunction=/FortUILibrary/MaterialFunctions/Basics/MF_UI_ChannelPackedDistanceField}
  n16 = NamedRerouteUsage()
  n17 = NamedRerouteUsage()
  n18 = Add(A=NamedRerouteUsage(), B=NamedRerouteUsage())
  n19 = NamedRerouteUsage()
  n20 = Add(
   A=Multiply(A=Max(A=@n16, B=@n17), B=ScalarParameter() {defaultValue=1, parameterName=OuterSpiralOpacity})
   B=Multiply(A=Add(A=@n18, B=@n19), B=ScalarParameter() {defaultValue=1, parameterName=Diamond Opacity})
  )
  n21 = Add(A=NamedRerouteUsage(), B=NamedRerouteUsage())
  Multiply(A=Saturate(None=Add(A=@n20, B=@n21)), B=ScalarParameter() {defaultValue=1, parameterName=GlobalOpacity})
  NamedRerouteDeclaration(None=VectorParameter() {parameterName=SpiralShadowColor, defaultValue=<rgba 0 0 0 1>}.RGB) {name=ShadowColor, nodeColor=<rgba 0 0 0 1>}
  n22 = TextureCoordinate()
  n23 = AppendVector(A=ComponentMask(None=@n22) {r=true}, B=Add(A=ComponentMask(None=@n22) {g=true}, B=Constant() {r=0.2}))
  n24 = LinearInterpolate(
   A=ScalarParameter() {defaultValue=0.5, parameterName=SpiralScale}
   B=Constant() {r=0.15}
   Alpha=ScalarParameter() {parameterName=IsPortalActive}
  )
  n25 = ScalarParameter() {parameterName=OuterSpiralRotation}
  n26 = Multiply(A=Constant() {r=-1}, B=Time())
  n27 = ScalarParameter() {parameterName=SpeedRotation}
  n28 = TextureSample(UVs=MaterialFunctionCall(
    UVs=MaterialFunctionCall(UVs=@n23, Texture Scale=@n24) {materialFunction=&ScaleUVsByCenter}.UVs
    "Rotation Angle (0-1)"=LinearInterpolate(A=@n25, B=@n26, Alpha=@n27) {constAlpha=0}
   ) {materialFunction=&CustomRotator}.Rotated Values) {samplerSource=SSM_Clamp_WorldGroupSettings, texture=/DemonMaskVoid/UI/Textures/T_UI_Reticle_DemonVoid}
  n29 = MaterialFunctionCall(RGBA Channels=@n28.R, Glow Min=Constant() {r=0.4}, Glow Max=Constant() {r=0.7}) {materialFunction=&MF_UI_ChannelPackedDistanceField}
  NamedRerouteDeclaration(None=@n29.Fill) {name=OuterSpiralShape, nodeColor=<rgba 0 0 0 1>}
  NamedRerouteDeclaration(None=SmoothStep(Value=@n15.R) {constMax=0.7, constMin=0.2}) {name=OuterShapeOutline, nodeColor=<rgba 0 0 0 1>}
  NamedRerouteDeclaration(None=Constant3Vector() {constant=<rgba 1 1 1 1>}) {name=MainColor, nodeColor=<rgba 0 0 0 1>}
  NamedRerouteDeclaration(None=Constant3Vector() {constant=<rgba 0.22 0.03 0.5 1>}) {name=OutlineColor, nodeColor=<rgba 0 0 0 1>}
  NamedRerouteDeclaration(None=VectorParameter() {parameterName=DiamondColor, defaultValue=<rgba 0.2882 0.0511 0.6536 1>}.RGB) {name=ActiveColorDiamond, nodeColor=<rgba 0 0 0 1>}
  n30 = TextureCoordinate()
  n31 = ComponentMask(None=@n30) {r=true}
  n32 = Add(A=ComponentMask(None=@n30) {g=true}) {constB=-0.28}
  n33 = TextureSample(UVs=MaterialFunctionCall(UVs=AppendVector(A=@n31, B=@n32), Texture Scale=ScalarParameter() {defaultValue=0.4, parameterName=ScaleDropLines}) {materialFunction=&ScaleUVsByCenter}.UVs) {samplerSource=SSM_Clamp_WorldGroupSettings, texture=&T_UI_Reticle_DemonVoid}
  NamedRerouteDeclaration(None=MaterialFunctionCall(RGBA Channels=@n33.B, Glow Min=Constant() {r=0.4}, Glow Max=Constant() {r=0.7}) {materialFunction=&MF_UI_ChannelPackedDistanceField}.Fill) {name=DropLines, nodeColor=<rgba 0 0 0 1>}
  n34 = Subtract(
   A=MaterialFunctionCall(RGBA Channels=@n28.R, Glow Min=Constant() {r=0.1}, Glow Max=Constant() {r=0.9}) {materialFunction=&MF_UI_ChannelPackedDistanceField}.Fill
   B=@n29.Fill
  )
  NamedRerouteDeclaration(None=Saturate(None=Multiply(A=@n34) {constB=5})) {name=OuterSpiralOutline, nodeColor=<rgba 0 0 0 1>}
  n35 = TextureSample(UVs=@n14.UVs) {samplerSource=SSM_Clamp_WorldGroupSettings, texture=&T_UI_Reticle_DemonVoid}
  n36 = Constant()
  n37 = Constant() {r=0.99}
  NamedRerouteDeclaration(None=Multiply(
    A=MaterialFunctionCall(RGBA Channels=@n35.G, Glow Min=@n36, Glow Max=@n37) {materialFunction=&MF_UI_ChannelPackedDistanceField}.Fill
    B=ScalarParameter() {parameterName=GlowShadowIntensity}
   )) {name=OuterShape, nodeColor=<rgba 0 0 0 1>}
  n38 = ScalarParameter() {defaultValue=0.05, parameterName=ShadowSpacing}
  n39 = Constant() {r=1}
  NamedRerouteDeclaration(None=Multiply(
    A=MaterialFunctionCall(RGBA Channels=@n28.R, Glow Min=@n38, Glow Max=@n39) {materialFunction=&MF_UI_ChannelPackedDistanceField}.Fill
    B=ScalarParameter() {defaultValue=0.3, parameterName=GlowShadowIntensity}
   ) {constB=0.5}) {name=OuterSpiralGlow, nodeColor=<rgba 0 0 0 1>}
  n40 = Constant()
  n41 = Constant() {r=1}
  NamedRerouteDeclaration(None=Multiply(
    A=MaterialFunctionCall(RGBA Channels=@n33.B, Glow Min=@n40, Glow Max=@n41) {materialFunction=&MF_UI_ChannelPackedDistanceField}.Fill
    B=Constant() {r=0.4}
   )) {name=DropLinesShadow, nodeColor=<rgba 0 0 0 1>}
  NamedRerouteDeclaration(None=VectorParameter() {parameterName=SpiralColor, defaultValue=<rgba 0.5853 0.3229 1 1>}.RGB) {name=ActiveColorSpiral, nodeColor=<rgba 0 0 0 1>}
  NamedRerouteDeclaration(None=MaterialFunctionCall(RGBA Channels=@n35.G, Glow Min=Constant() {r=0.35}, Glow Max=Constant() {r=0.5}) {materialFunction=&MF_UI_ChannelPackedDistanceField}.Fill) {name=OuterShapeSolid, nodeColor=<rgba 0 0 0 1>}
 ]
}
