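# Reticle extension widget for the Demon Mask Void teleport weapon, showing throw, teleport and cooldown states.
# On Construct it casts the equipped weapon and binds its projectile, marker and cooldown delegates, then unbinds them on Destruct.
# Those events play widget animations forward or in reverse, driving material parameters on a single M_UI_VoidMaskReticle image.
# Marker updates branch on validity to reverse the throw or unavailable animation; primary-fire cooldown handlers only print debug strings.
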WidgetBlueprint /DemonMaskVoid/UI/WBP_DemonMask_Void
animations [
 Anim_Teleport_Throw {
  binding_count=1
  length_seconds=0.7
  object=&self:Anim_Teleport_Throw
  sequence {bindings [Reticle {
     id=#1
     possessed_class=Image
     tracks [{
       type=MovieSceneWidgetMaterialTrack
       sections [{
         type=MovieSceneParameterSection
         end_frame=14
         start_frame=0
         channels [
          SpiralScale {type=MovieSceneScriptingFloatChannel, keys [[0,0.5],[5,0.56],[10,0.45],[14,0.5]]}
          Switch {type=MovieSceneScriptingFloatChannel, keys [[0,0],[10,0.5],[14,1]]}
          ShadowSpacing {type=MovieSceneScriptingFloatChannel, keys [[0,0.3],[13.9997,0]]}
          GlowShadowIntensity {type=MovieSceneScriptingFloatChannel, keys [[0,0.2],[13.9997,1]]}
          InnerSpiralRotation {type=MovieSceneScriptingFloatChannel, keys [[0,0],[5,-0.2158],[14,-0.5]]}
          OuterSpiralRotation {type=MovieSceneScriptingFloatChannel, keys [[0,0],[5,-0.045],[14,-0.5]]}
          InnerSpiralScale {type=MovieSceneScriptingFloatChannel, keys [[0,0.5],[14,0.25]]}
          InnerSpiralOpacity {type=MovieSceneScriptingFloatChannel, keys [[0,0],[5,0.5],[14,0.8]]}
         ]
        }]
      }]
    }]}
 }
 Anim_Teleport_UseTeleport {
  binding_count=1
  length_seconds=0.6
  object=&self:Anim_Teleport_UseTeleport
  sequence {bindings [Reticle {
     id=#1
     possessed_class=Image
     tracks [{
       type=MovieSceneWidgetMaterialTrack
       sections [{
         type=MovieSceneParameterSection
         end_frame=12
         start_frame=0
         channels [
          IsPortalActive {type=MovieSceneScriptingFloatChannel, keys [[0,1],[2,1.12],[12,0]]}
          OuterSpiralRotation {type=MovieSceneScriptingFloatChannel, keys [[0,0],[12,-0.5]]}
          ScaleXDiamond {type=MovieSceneScriptingFloatChannel, keys [[0,0.6],[3,0.8696],[6,0.2296]]}
          ScaleYDiamond {type=MovieSceneScriptingFloatChannel, keys [[0,0.6],[3,0.69],[6,0.05]]}
          GlowShadowIntensity {type=MovieSceneScriptingFloatChannel, keys [[0,0.5],[2,1.536],[6,0.3]]}
          Diamond Opacity {type=MovieSceneScriptingFloatChannel, keys [[4,1],[6,0]]}
          SpeedRotation {type=MovieSceneScriptingFloatChannel, const [0,1]}
          SpiralColorR {type=MovieSceneScriptingFloatChannel, const [0.67,1]}
          SpiralColorG {type=MovieSceneScriptingFloatChannel, const [0.35,1]}
          SpiralColorB {type=MovieSceneScriptingFloatChannel, const [0.95,1]}
          SpiralColorA {type=MovieSceneScriptingFloatChannel}
          DiamondColorR {type=MovieSceneScriptingFloatChannel, keys [[0,0.4],[3,0.5],[5,0.19]]}
          DiamondColorG {type=MovieSceneScriptingFloatChannel, keys [[0,0.05],[3,0.2],[5,-0.01]]}
          DiamondColorB {type=MovieSceneScriptingFloatChannel, keys [[0,0.95],[3,1],[5,1]]}
          DiamondColorA {type=MovieSceneScriptingFloatChannel, const [1,1]}
         ]
        }]
      }]
    }]}
 }
 Anim_Teleport_CooldownEnded {
  binding_count=1
  length_seconds=0.3
  object=&self:Anim_Teleport_CooldownEnded
  sequence {bindings [Reticle {
     id=#1
     possessed_class=Image
     tracks [{
       type=MovieSceneWidgetMaterialTrack
       sections [{
         type=MovieSceneParameterSection
         end_frame=6
         start_frame=0
         channels [
          SpiralScale {type=MovieSceneScriptingFloatChannel, keys [[0,0.5],[1,0.6],[4,0.45],[6,0.5]]}
          Switch {type=MovieSceneScriptingFloatChannel, const [0,2]}
          DropLinesOpacity {type=MovieSceneScriptingFloatChannel, keys [[-0.0003,1],[0,0.2],[6,1]]}
          ShadowSpacing {type=MovieSceneScriptingFloatChannel, keys [[-0.0003,0.3],[0,1],[6,0.3]]}
          GlowShadowIntensity {type=MovieSceneScriptingFloatChannel, const [1,2]}
          InnerSpiralRotation {type=MovieSceneScriptingFloatChannel, const [-1,2]}
          OuterSpiralRotation {type=MovieSceneScriptingFloatChannel, keys [[-0.0003,-1],[6,-1.5]]}
          InnerSpiralScale {type=MovieSceneScriptingFloatChannel, const [0.63,2]}
          OuterSpiralOpacity {type=MovieSceneScriptingFloatChannel, keys [[0,0.2],[6,1]]}
         ]
        }]
      }]
    }]}
 }
 Anim_Teleport_OnCooldown {
  binding_count=1
  length_seconds=0.25
  object=&self:Anim_Teleport_OnCooldown
  sequence {bindings [Reticle {
     id=#1
     possessed_class=Image
     tracks [{
       type=MovieSceneWidgetMaterialTrack
       sections [{
         type=MovieSceneParameterSection
         end_frame=5
         start_frame=-1
         channels [
          SpiralScale {type=MovieSceneScriptingFloatChannel, const [0.5,2]}
          Switch {type=MovieSceneScriptingFloatChannel, const [0,2]}
          DropLinesOpacity {type=MovieSceneScriptingFloatChannel, keys [[-0.0003,1],[-0.0003,0.2],[5,0.2]]}
          ShadowSpacing {type=MovieSceneScriptingFloatChannel, const [1,2]}
          GlowShadowIntensity {type=MovieSceneScriptingFloatChannel, const [1,2]}
          InnerSpiralRotation {type=MovieSceneScriptingFloatChannel, const [-1,2]}
          OuterSpiralRotation {type=MovieSceneScriptingFloatChannel, const [-1,2]}
          InnerSpiralScale {type=MovieSceneScriptingFloatChannel, const [0.63,2]}
          OuterSpiralOpacity {type=MovieSceneScriptingFloatChannel, const [0.2,2]}
         ]
        }]
      }]
    }]}
 }
 Anim_TeleportThrow {
  binding_count=1
  length_seconds=0.6
  object=&self:Anim_TeleportThrow
  sequence {bindings [Reticle {
     id=#2
     possessed_class=Image
     tracks [{
       type=MovieSceneWidgetMaterialTrack
       sections [{
         type=MovieSceneParameterSection
         end_frame=12
         start_frame=0
         channels [
          IsPortalActive {type=MovieSceneScriptingFloatChannel, keys [[0,0],[3,-0.24],[6,0],[10,1]]}
          OuterSpiralRotation {type=MovieSceneScriptingFloatChannel, keys [[0,0],[9,-0.7]]}
          ScaleXDiamond {type=MovieSceneScriptingFloatChannel, keys [[0,0],[7,0.166],[10,0.6]]}
          ScaleYDiamond {type=MovieSceneScriptingFloatChannel, keys [[0,0],[7,0.02],[10,0.4],[12,0.6]]}
          SpeedRotation {type=MovieSceneScriptingFloatChannel, keys [[9,0],[10,0.3]]}
          GlowShadowIntensity {type=MovieSceneScriptingFloatChannel, keys [[0,0.3],[8,0.3],[9,0.5]]}
          Diamond Opacity {type=MovieSceneScriptingFloatChannel, keys [[6,0],[7,1]]}
          SpiralColorR {type=MovieSceneScriptingFloatChannel, keys [[9,1],[10,0.67]]}
          SpiralColorG {type=MovieSceneScriptingFloatChannel, keys [[9,1],[10,0.35]]}
          SpiralColorB {type=MovieSceneScriptingFloatChannel, keys [[9,1],[10,0.95]]}
          SpiralColorA {type=MovieSceneScriptingFloatChannel, const [1,1]}
          DiamondColorR {type=MovieSceneScriptingFloatChannel, const [0.4,1]}
          DiamondColorG {type=MovieSceneScriptingFloatChannel, const [0.05,1]}
          DiamondColorB {type=MovieSceneScriptingFloatChannel, const [0.95,1]}
          DiamondColorA {type=MovieSceneScriptingFloatChannel, const [1,1]}
         ]
        }]
      }]
    }]}
 }
 Anim_TeleportUnavailable {
  binding_count=1
  length_seconds=0.25
  object=&self:Anim_TeleportUnavailable
  sequence {bindings [Reticle {
     id=#3
     possessed_class=Image
     tracks [{
       type=MovieSceneWidgetMaterialTrack
       sections [{
         type=MovieSceneParameterSection
         end_frame=5
         start_frame=0
         channels [
          IsPortalActive {type=MovieSceneScriptingFloatChannel, const [1,1]}
          GlobalOpacity {type=MovieSceneScriptingFloatChannel, keys [[0,1],[5,0.3]]}
          ScaleXDiamond {type=MovieSceneScriptingFloatChannel, const [0.6,1]}
          ScaleYDiamond {type=MovieSceneScriptingFloatChannel, const [0.6,1]}
          GlowShadowIntensity {type=MovieSceneScriptingFloatChannel, const [0.5,1]}
          SpiralColorR {type=MovieSceneScriptingFloatChannel, const [0.67,1]}
          SpiralColorG {type=MovieSceneScriptingFloatChannel, const [0.35,1]}
          SpiralColorB {type=MovieSceneScriptingFloatChannel, const [0.95,1]}
          SpiralColorA {type=MovieSceneScriptingFloatChannel, const [1,1]}
          DiamondColorR {type=MovieSceneScriptingFloatChannel, const [0.4,1]}
          DiamondColorG {type=MovieSceneScriptingFloatChannel, const [0.05,1]}
          DiamondColorB {type=MovieSceneScriptingFloatChannel, const [0.95,1]}
          DiamondColorA {type=MovieSceneScriptingFloatChannel, const [1,1]}
         ]
        }]
      }]
    }]}
 }
 Anim_TeleportCooldown {
  binding_count=1
  length_seconds=0.2
  object=&self:Anim_TeleportCooldown
  sequence {bindings [Reticle {
     id=#4
     possessed_class=Image
     tracks [{
       type=MovieSceneWidgetMaterialTrack
       sections [{
         type=MovieSceneParameterSection
         end_frame=4
         start_frame=0
         channels [GlobalOpacity {type=MovieSceneScriptingFloatChannel, keys [[0,1],[4,0.5]]}]
        }]
      }]
    }]}
 }
]
graphs [EventGraph {nodes [
   OnProjectileCreated!() -> (OutputDelegate, then, DemonMaskProjectile:B Prj Demon Mask Void Object Reference) {is=title K2Node_CustomEvent}
   > PlayAnimationForward(InAnimation=Get Anim_TeleportThrow() -> (Anim_TeleportThrow:Widget Animation Object Reference))
   OnPrimaryFireCooldown() {is=title K2Node_CustomEvent}
   > PrintString!(execute, InString=CooldownStarted, bPrintToScreen, bPrintToLog, TextColor, Duration, Key) -> (then)
   OnPrimaryFireCooldownEnded() {is=title K2Node_CustomEvent}
   > PrintString(InString=FireCooldownEnded)
   OnTeleportedCooldown() {is=title K2Node_CustomEvent}
   > PlayAnimationForward(InAnimation=Get Anim_Teleport_UseTeleport() -> (Anim_Teleport_UseTeleport:Widget Animation Object Reference))
   > PlayAnimationForward(InAnimation=Get Anim_TeleportCooldown() -> (Anim_TeleportCooldown:Widget Animation Object Reference))
   OnTeleportedCooldownEnded() {is=title K2Node_CustomEvent}
   > PlayAnimationReverse(InAnimation=Get Anim_TeleportCooldown())
   n1 = OnBGAMarkerUpdated!() -> (OutputDelegate, then, TeleportMarker BGA:BGA Demon Mask Void Teleport Base Object Reference) {is=title K2Node_CustomEvent}
   > n2 = Is Valid(InputObject=@n1.TeleportMarker BGA)
   PlayAnimationReverse(execute=@n2.Is Not Valid, InAnimation=Get Anim_TeleportThrow())
   PlayAnimationReverse(execute=@n2.Is Valid, InAnimation=Get Anim_TeleportUnavailable() -> (Anim_TeleportUnavailable:Widget Animation Object Reference))
   Event Destruct()
   n3 = Get WeaponRef() -> (WeaponRef:B Demon Mask Void Teleport Object Reference)
   > Unbind Event from On New Projectile Created and Valid(self:B Demon Mask Void Teleport Object Reference=@n3, Delegate=Create Event()) {is=title K2Node_RemoveDelegate}
   > Unbind Event from On Primary Fire Cooldown(self:B Demon Mask Void Teleport Object Reference=@n3, Delegate=Create Event()) {is=title K2Node_RemoveDelegate}
   > Unbind Event from On Primary Fire Cooldown Ended(self:B Demon Mask Void Teleport Object Reference=@n3, Delegate=Create Event()) {is=title K2Node_RemoveDelegate}
   > Unbind Event from On Teleported Cooldown(self:B Demon Mask Void Teleport Object Reference=@n3, Delegate=Create Event()) {is=title K2Node_RemoveDelegate}
   > Unbind Event from On Teleported Cooldown Ended(self:B Demon Mask Void Teleport Object Reference=@n3, Delegate=Create Event()) {is=title K2Node_RemoveDelegate}
   > Unbind Event from On New BGACreated and Valid(self:B Demon Mask Void Teleport Object Reference=@n3, Delegate=Create Event()) {is=title K2Node_RemoveDelegate}
   Event Construct()
   > n4 = Cast To B_DemonMask_Void_Teleport!(
    execute
    Object=GetEquippedWeapon!(self:Fort Weapon Reticle Extension Widget Base Object Reference) -> (ReturnValue:Fort Weapon Object Reference)
   ) -> (then, CastFailed, AsB Demon Mask Void Teleport:B Demon Mask Void Teleport Object Reference) {is=title K2Node_DynamicCast}
   > n5 = Set WeaponRef(WeaponRef:B Demon Mask Void Teleport Object Reference=@n4) -> (Output_Get:B Demon Mask Void Teleport Object Reference)
   > Bind Event to On New Projectile Created and Valid(self:B Demon Mask Void Teleport Object Reference=@n5, Delegate=Create Event()) {is=title K2Node_AddDelegate}
   > Bind Event to On Primary Fire Cooldown(self:B Demon Mask Void Teleport Object Reference=@n5, Delegate=Create Event()) {is=title K2Node_AddDelegate}
   > Bind Event to On Teleported Cooldown(self:B Demon Mask Void Teleport Object Reference=@n5, Delegate=Create Event()) {is=title K2Node_AddDelegate}
   > Bind Event to On Teleported Cooldown(self=@n5, Delegate=Create Event())
   > Bind Event to On Teleported Cooldown Ended(self:B Demon Mask Void Teleport Object Reference=@n5, Delegate=Create Event()) {is=title K2Node_AddDelegate}
   > Bind Event to On New BGACreated and Valid(self:B Demon Mask Void Teleport Object Reference=@n5, Delegate=Create Event()) {is=title K2Node_AddDelegate}
   Get Anim_TeleportThrow()
  ]}]
variables [
 Anim Teleport Cooldown Ended {type=WidgetAnimation, default=null}
 WeaponRef {type=B_DemonMask_Void_Teleport_C, default=null}
]
widgets {
 info {
  inheritedWidgetCount=1
  parentClass=/Script/FortniteUI.FortWeaponReticleExtensionWidgetBase
  rootWidgetClass=&Overlay
  widgetCount=3
 }
 widgets [
  Overlay_0 {widgetClassPath=&Overlay, widget=Overlay_0}
  Reticle {
   widgetClassPath=&Image
   bIsVariable=true
   parent=Overlay_0
   widget=Reticle
   properties {brush {
     bIsDynamicallyLoaded=false
     drawAs=Image
     imageType=FullColor
     mirroring=NoMirror
     resourceName=None
     resourceObject=/DemonMaskVoid/UI/Materials/M_UI_VoidMaskReticle
     tiling=NoTile
     imageSize=<130 130>
     margin=<ltrb 0 0 0 0>
     outlineSettings {
      bUseBrushTransparency=false
      roundingType=HalfHeightRadius
      width=0
      color {colorUseRule=UseColor_Specified, specifiedColor=<rgba 0 0 0 0>}
      cornerRadii=<0 0 0 0>
     }
     tintColor {colorUseRule=UseColor_Specified, specifiedColor=<rgba 1 1 1 1>}
     uVRegion {bIsValid=false, max=<0 0>, min=<0 0>}
    }}
   slot {class=OverlaySlot, properties {padding=<ltrb 0 30 0 0>}}
  }
  TutorialHighlight {widgetClassPath=&TutorialHighlight}
 ]
}
